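# Niagara trail effect for the BottleCap Challenge emote's cap, drawing short-lived, fading wind ribbons along a skeletal mesh.
# A single contrail emitter spawns particles continuously, assigns spawn groups that select the bone or socket on User.SkeletalMesh, and uses that bone as ribbon ID.
# Particles lock to their bone, with a Z offset for lower groups, while width and alpha taper over life.
# Velocity and force modules are disabled.
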NiagaraSystem /BRCosmetics/Animation/Game/MainPlayer/Emotes/BottleCap_Challenge/VFX/NS_BottleCap_Challenge_CapTrail
dependencies {
 usedDataInterfaces [&NiagaraDataInterfaceColorCurve, &NiagaraDataInterfaceCurve]
 usedDynamicInputs []
 usedModules [/Niagara/Modules/System/SystemLifeCycle, &SystemState]
}
emitters {FN_Contrails_Ribbon_Wind {
  emitterSpawnScript [&FN_PreserveSkydiveUserVariables() {module=PreserveUserVariables}]
  emitterUpdateScript [
   &EmitterState(Enable Distance Culling=true, Module.bUseMaxDistance=true, Module.bUseMinDistance=false, Scalability Mode=::NewEnumerator1)
   &SpawnRate(SpawnRate=120, Use Spawn Probability=false)
  ]
  particleSpawnScript [
   /Game/Effects/Niagara/Modules/Groups/FN_SetSpawnGroup_v1/FN_SetSpawnGroup(Override SpawnGroup=true, SpawnGroup:NiagaraInt32=&SpawnGroup())
   &InitializeRibbon(
    Color=&ScaleLinearColorByCurve(
     InitialValue=<rgba 1 1 1 1>
     ScaleCurve=DataInterface() {propertyValues {AlphaCurve {keys [<at 0 0.035>]}, BlueCurve {keys []}, GreenCurve {keys []}, RedCurve {keys []}}}
     CurveIndex=$Particles.NormalizedAge
    )
    Lifetime=&Add_Float(A=0.35, B=0)
    Module.Write Color=true
    Module.Write Lifetime=true
    Module.Write Mass=false
    Module.Write Material Random=true
    Module.Write Position=false
    Module.Write RibbonTwist=false
    Module.Write RibbonWidth=true
    Position:Vector3f=&SimulationPosition()
    Ribbon Width=25
   )
   &SampleSkeletalMeshSkeleton(
    Mesh:NiagaraDataInterfaceSkeletalMesh=$User.SkeletalMesh
    Specific Bone Index:NiagaraInt32=$Particles.SpawnGroup
    Specific Socket Index:NiagaraInt32=$Particles.SpawnGroup
   )
   /Niagara/Modules/Spawn/Location/SkeletalMeshSkeletonLocation(
    Inherit Velocity From Mesh=0
    Module.Inherit Velocity=true
    Module.Use Sampled Bone as Ribbon ID=true
    Sampled Position Offset:Vector3f=&MakeVector(X=0, Y=0, Z=HlslExpression() {hlslExpression="Particles.SpawnGroup < 5 ? 10 : 0"})
   )
   &AddVelocityInCone(Cone Axis=<0 0 1>, Module.Use Velocity Falloff On Cone Axis=false, Velocity Strength=20) {enabled=false}
   &AddVelocity(Velocity=&FN_FrontEndVector3Modifier(Gameplay=<0 0 0>, Front End=<0 0 500>)) {enabled=false, module=AddVelocity001}
   &AddVelocity(Velocity=&FN_FrontEndVector3Modifier(Gameplay=<0 0 0>, Front End=<0 0 500>)) {enabled=false}
  ]
  particleUpdateScript [
   &ParticleState()
   &SolveForcesAndVelocity(Acceleration Limit=0, Clamp Velocity=false, Limit Acceleration=true) {enabled=false}
   /Game/Effects/Niagara/Modules/LockingParticles/FN_LockParticleToBone(SkeletalMesh:NiagaraDataInterfaceSkeletalMesh=$User.SkeletalMesh)
   /SetVariables(Particles.RibbonWidth=&Multiply_Float(
     A=$Particles.Initial.RibbonWidth
     B=&FloatFromCurve(FloatCurve=DataInterface() {propertyValues {Curve {keys [
          {arriveTangent=1.695, interpMode=RCIM_Cubic, leaveTangent=1.695, tangentMode=RCTM_User, time=-0.0046, value=0}
          {arriveTangent=0.1999, interpMode=RCIM_Cubic, leaveTangent=0.1999, time=0.5, value=1}
          <at 0.9877 0.1983>
         ]}}})
    )) {enabled=false, isSetParameters=true}
   &Material(Base Texture Time:NiagaraFloat=&Multiply_Float(A=-1, B=$Emitter.Age)) {module=Material001}
   /Game/Effects/Niagara/Modules/Materials/MaterialComponentMask(Blue=0.75, Red=0.5) {enabled=false}
   &ScaleColor(
    Scale Alpha=&FloatFromCurve(FloatCurve=DataInterface() {propertyValues {Curve {keys [
         {arriveTangent=-1.631, interpMode=RCIM_Cubic, leaveTangent=-1.631, tangentMode=RCTM_User, time=0, value=0.65}
         {interpMode=RCIM_Cubic, tangentMode=RCTM_User, time=0.99, value=0}
        ]}}})
    ScaleA=true
    ScaleRGB=true
   )
   &Material()
   &ScaleRibbonWidth(Ribbon Width Scale=&FloatFromCurve(FloatCurve=DataInterface() {propertyValues {Curve {keys [<at -0.0046 1.0113>,<at 1.0011 0.002>]}}}))
  ]
  rendererClasses [&NiagaraRibbonRendererProperties, &NiagaraMeshRendererProperties]
  renderers [
   {rendererClass=&NiagaraRibbonRendererProperties}
   {rendererClass=&NiagaraMeshRendererProperties, rendererIndex=1}
   {rendererClass=&NiagaraRibbonRendererProperties, rendererIndex=2}
  ]
 }}
summary {
 systemName=NS_BottleCap_Challenge_CapTrail
 emitters [{
   emitterName=FN_Contrails_Ribbon_Wind
   rendererClasses [&NiagaraRibbonRendererProperties, &NiagaraMeshRendererProperties]
  }]
 userVariables [
  User.bIsFrontEndPreview {type=NiagaraBool}
  User.PlayerSpeed {type=NiagaraFloat, defaultValue=0, description=""}
  User.SkeletalMesh {
   type=NiagaraDataInterfaceSkeletalMesh
   defaultValue {
    struct=NiagaraExt_VariableValue_DataInterface
    value {dataInterface=&self:NiagaraDataInterfaceSkeletalMesh_0}
   }
  }
 ]
}
